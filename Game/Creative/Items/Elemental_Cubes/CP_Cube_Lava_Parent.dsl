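; CP_Cube_Lava_Parent is a Blueprint, by its name the parent class for lava elemental cube items in Creative.
; Its ShouldUsePhysicalMaterialForFootstep function returns the bIsLavaSurface user option, so that flag decides whether footsteps use the physical material.
; The construction script is empty.
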(fn ShouldUsePhysicalMaterialForFootstep
  (return (Variables|UserOptions|Basic|GetbIsLavaSurface)))

(fn ConstructionScript)
